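; Car wash roller prop that survives losing the building piece it is attached to if another is within reach.
; At BeginPlay its mesh ignores one game trace channel, so items like build-a-bunker do not destroy it while it stays damageable.
; On detach, a box trace sized to the mesh looks for other building static mesh actors to attach to, skipping the previous one.
; If none is found, it destroys itself.
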(event EventBeginPlay
  ; This will make it so  certain building items will not destroy this actor, but still make it damageable (ex. build-a-bunker)
  (|CallParentFunction)
  (Collision|SetCollisionResponsetoChannel (Variables|Debug|GetStaticMeshComponent) "ECC_GameTraceChannel3" "ECR_Ignore"))

(event Attachment|EventOnDetachFromBuilding
  (params (AttachedToBuilding :type Building%20Actor%20Object%20Reference) (bBeingDestroyed :type Boolean))
  (bind _staticmeshcomponent (Variables|Debug|GetStaticMeshComponent))
  ; This actor visually appears to be attached to multiple actors.  When the current one it's attached to is destroyed, see if there is another one to attach to and use it.
  (Variables|Default|SetPreviousAttachedBuilding AttachedToBuilding)
  (Variables|Default|SetFoundNewAttachActor false)
  (bind _returnvalue (FortPhysicsSimulationLibrary|GetBoxExtent _staticmeshcomponent))
  (bind _outhits (Collision|MultiBoxTraceForObjects (Transformation|GetWorldLocation _staticmeshcomponent) (+ (Transformation|GetWorldLocation _staticmeshcomponent) "0.000000,0.000000,10.000000") _returnvalue (Transformation|GetWorldRotation _staticmeshcomponent) (Utilities|Array|MakeArray "ObjectTypeQuery1") false (Utilities|Array|MakeArray self) "None" true "(R=1.000000,G=0.000000,B=0.000000,A=1.000000)" "(R=0.000000,G=1.000000,B=0.000000,A=1.000000)" 3e+01))
  (bind _array_element (Utilities|Array|ForEachLoopwithBreak _outhits)
    (:LoopBody
      (if (!= (Variables|Default|GetPreviousAttachedBuilding) (.HitActor (Collision|BreakHitResult _array_element)))
        (bind _asbuilding_static_mesh (Utilities|Casting|CastToBuildingSMActor (.HitActor (Collision|BreakHitResult _array_element))))
        (Attachment|AttachBuildingActortoMe _asbuilding_static_mesh self true)
        (Variables|Default|SetFoundNewAttachActor true)))
    (:Completed
      _))
  (if (not (Variables|Default|GetFoundNewAttachActor))
    (Actor|DestroyActor)))

(fn ConstructionScript
  (|CallParentFunction))
