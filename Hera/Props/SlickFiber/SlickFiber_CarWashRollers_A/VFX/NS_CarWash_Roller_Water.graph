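# Niagara water spray effect for the car wash roller prop, layering soft mist, splash sprites and small GPU droplets.
# TopSmoke drifts and fades, while two TopSplash emitters launch gravity-pulled cone bursts, the GPU one adding vortex force and speed-based stretching.
# User.FallWidth scales only spawn rates, as width offsets are disabled; most other user parameters are unreferenced by the emitters.
# System-level time-of-day color and alpha tint every emitter, and all use distance culling.
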NiagaraSystem /Hera/Props/SlickFiber/SlickFiber_CarWashRollers_A/VFX/NS_CarWash_Roller_Water
dependencies {
 usedDataInterfaces [
  &NiagaraDataInterfaceCurve
  &NiagaraDataInterfaceArrayFloat3
  &NiagaraDataInterfaceArrayFloat
  &NiagaraDataInterfaceCamera
 ]
 usedDynamicInputs [
  &Subtract_Float
  &Abs_Float
  &DotProduct
  &ReturnCameraVector
  &FN_Select_Float_from_Time_Of_Day
  &MultiplyLinearColor
  &FN_Select_Color_from_Time_Of_Day
  &MakeVector4_FromFloat
 ]
 usedModules [
  &SystemState
  &self:SystemSpawnScript.SystemScriptSource.SystemScriptGraph.NiagaraNodeAssignment_1.SetVariables_#1
  &self:SystemSpawnScript.SystemScriptSource.SystemScriptGraph.NiagaraNodeAssignment_0.SetVariables_#2
  &self:SystemSpawnScript.SystemScriptSource.SystemScriptGraph.NiagaraNodeAssignment_2.SetVariables_#3
  &self:SystemSpawnScript.SystemScriptSource.SystemScriptGraph.NiagaraNodeAssignment_4.SetVariables_#4
 ]
}
emitters {
 TopSmoke {
  emitterUpdateScript [
   &EmitterState(
    Enable Distance Culling=true
    MaxDistance=20000
    Module.bUseMaxDistance=true
    Module.bUseMinDistance=false
    Scalability Mode=::NewEnumerator1
   )
   &SpawnRate(
    SpawnRate=&Multiply_Float(A=1, B=&Divide_Float(A=$User.FallWidth, B=850))
    Use Spawn Probability=false
   )
  ]
  particleSpawnScript [
   &PartitionParticles() {enabled=false}
   &InitializeParticle(
    Color=<rgba 1 1 1 0.35>
    Lifetime Max=2.2
    Lifetime Min=2
    Lifetime Mode=::NewEnumerator1
    Sprite Size Mode=::NewEnumerator3
    Uniform Sprite Size=&Lerp_Float(
     A=400
     B=600
     Alpha=&FloatFromCurve(
      FloatCurve=DataInterface() {propertyValues {Curve {keys [
          {interpMode=RCIM_Cubic, leaveTangent=0.1305, tangentMode=RCTM_Break, time=0, value=0}
          {arriveTangent=1, interpMode=RCIM_Cubic, leaveTangent=1, time=0.7, value=0.3}
          {arriveTangent=2.0211, interpMode=RCIM_Cubic, tangentMode=RCTM_Break, time=1, value=1}
         ]}}}
      CurveIndex=&RandomRangeFloat(Minimum=0, Maximum=1)
     )
    )
    UsePositionOffset=false
    Write Material Random=true
   )
   &ShapeLocation(Box Size=<50 50 150>, MaskSpawnGroup=false, Override Seed=false, Shape Primitive=::NewEnumerator2, Surface Only=false, UseOwnerScale=true) {module=ShapeLocation001}
   /SetVariables(Particles.Position=&AddVectorToPosition(
     Position=$Particles.Position
     Vector=&Multiply_VectorByFloat(
      Vector=&TransformVector(Vector=<1 0 0>, Destination Space=::Simulation)
      Float=&Multiply_Float(
       A=&Lerp_Float(
        A=0
        B=&Divide_Float(A=$User.FallWidth, B=2)
        Alpha=&FloatFromCurve(
         FloatCurve=DataInterface() {propertyValues {Curve {keys [
             {interpMode=RCIM_Cubic, leaveTangent=3.3639, tangentMode=RCTM_Break, time=0, value=0}
             {arriveTangent=-0.0122, interpMode=RCIM_Cubic, tangentMode=RCTM_Break, time=1, value=1}
            ]}}}
         CurveIndex=&RandomRangeFloat(Minimum=0, Maximum=1)
        )
       )
       B=&Lerp_Float(
        A=-1
        B=1
        Alpha=&MakeFloatFromInt(Int=&RandomRangeInt(Minimum=0, Maximum=1))
       )
      )
     )
    )) {enabled=false, isSetParameters=true}
   /SetVariables() {isSetParameters=true}
   &AddVelocity(
    Velocity=<150 0 0>
    Velocity Speed Scale=&SelectFloatFromArray(Direct Array Index=$Particles.PartitionParticles.Partition, Float Selection Array=DataInterface() {propertyValues {FloatData [-1,1]}})
   )
   &CameraOffset(Camera Offset Amount=100)
  ]
  particleUpdateScript [
   &ParticleState()
   &SpriteRotationRate(Rotation Rate=-50)
   &ScaleColor(
    Scale Alpha=&FloatFromCurve(FloatCurve=DataInterface() {propertyValues {Curve {keys [
         {interpMode=RCIM_Cubic, tangentMode=RCTM_User, time=0, value=0}
         {interpMode=RCIM_Cubic, tangentMode=RCTM_User, time=0.5, value=1}
         {tangentMode=RCTM_User, time=1, value=0}
        ]}}})
    ScaleA=true
    ScaleRGB=true
   )
   &GravityForce(Gravity=&RandomRangeVector(Minimum=<0 0 -50>, Maximum=<0 0 100>))
   &Drag(Drag=0.1636, Use Linear Drag=true, Use Rotational Drag=true)
   &SolveForcesAndVelocity(Clamp Velocity=false, Limit Acceleration=false)
   &DynamicMaterialParameters(
    Index 0 Param 1=&RandomRangeFloat(Minimum=0, Maximum=1)
    Module.Param0WriteEnabled=true
    Module.Param1WriteEnabled=true
    Module.Param2WriteEnabled=true
    Module.Param3WriteEnabled=true
   )
   &SubUVAnimation() {enabled=false, module=SubUVAnimation001}
   &ScaleColor(
    Scale Alpha=&FNTimeOfDayScalar(Night Scale=0.5, Morning Scale=0.2, Evening Scale=0.3)
    Scale RGB=&MakeVectorFromLinearColorRGB(Color=&FNTimeOfDayColor(Night Color=<rgba 0.6 0.8 1 1>, Morning Color=<rgba 1 0.625 0.55 1>, Day Color=<rgba 1 1 1 1>, Evening Color=<rgba 1 0.5 0.4 1>))
    ScaleA=true
    ScaleRGB=true
   ) {enabled=false, module=ScaleColor001}
   &ScaleColor(
    Scale Alpha=$System.TOD_Alpha
    Scale RGB=&MakeVectorFromLinearColorRGB(Color=$System.TOD_Color)
    ScaleA=true
    ScaleRGB=true
   ) {module=ScaleColor002}
   &KillParticles(Kill Particles=&SetBoolByFloatComparison(A=$Particles.NormalizedAge, B=0.95))
  ]
  rendererClasses [&NiagaraSpriteRendererProperties]
  renderers [{rendererClass=&NiagaraSpriteRendererProperties}]
 }
 TopSplash {
  emitterUpdateScript [
   &EmitterState(
    Enable Distance Culling=true
    Life Cycle Mode=::NewEnumerator1
    Loop Duration=2
    MaxDistance=10000
    Module.bUseMaxDistance=true
    Module.bUseMinDistance=false
    Scalability Mode=::NewEnumerator1
    UseLoopDelay=false
   )
   &SpawnRate(
    SpawnRate=&Multiply_Float(A=20, B=&Divide_Float(A=$User.FallWidth, B=850))
    Use Spawn Probability=false
   )
  ]
  particleSpawnScript [
   &PartitionParticles() {enabled=false}
   &InitializeParticle(
    Color Maximum=<rgba 1 1 1 1.5>
    Color Minimum=<rgba 1 1 1 1>
    Color Mode=::NewEnumerator2
    Lifetime Max=1
    Lifetime Min=0.6
    Lifetime Mode=::NewEnumerator1
    Mass Mode=::NewEnumerator3
    Sprite Rotation Mode=::NewEnumerator2
    Sprite Size Mode=::NewEnumerator1
    Uniform Sprite Size Max=100
    Uniform Sprite Size Min=50
    UsePositionOffset=false
    Write Material Random=true
   )
   &ShapeLocation(Box Size=<0 50 150>, MaskSpawnGroup=false, Override Seed=false, Shape Primitive=::NewEnumerator2, Surface Only=false, UseOwnerScale=true)
   /SetVariables(Particles.Position=&AddVectorToPosition(
     Position=$Particles.Position
     Vector=&Multiply_VectorByFloat(
      Vector=&TransformVector(Vector=<1 0 0>, Destination Space=::Simulation)
      Float=&Multiply_Float(
       A=&Lerp_Float(
        A=0
        B=&Divide_Float(A=$User.FallWidth, B=2)
        Alpha=&FloatFromCurve(
         FloatCurve=DataInterface() {propertyValues {Curve {keys [
             {interpMode=RCIM_Cubic, leaveTangent=3.3639, tangentMode=RCTM_Break, time=0, value=0}
             {arriveTangent=-0.0122, interpMode=RCIM_Cubic, tangentMode=RCTM_Break, time=1, value=1}
            ]}}}
         CurveIndex=&RandomRangeFloat(Minimum=0, Maximum=1)
         Scale Curve=0.8
        )
       )
       B=&Lerp_Float(
        A=-1
        B=1
        Alpha=&MakeFloatFromInt(Int=&RandomRangeInt(Minimum=0, Maximum=1))
       )
      )
     )
    )) {enabled=false, isSetParameters=true, module=SetVariables_#5}
   &AddVelocity(
    Cone Angle=150
    Cone Axis=&SelectVectorFromArray(
     Direct Array Index=$Particles.PartitionParticles.Partition
     Vector Selection Array=DataInterface() {propertyValues {FloatData [<-1 0 0.2>, <1 0 0.2>]}}
    )
    Distribution Along Cone Axis=0.25
    Inner Cone Angle=50
    Module.UseConeFalloff=true
    Module.UseVelDistribution=true
    Override Seed=false
    Speed Falloff From Cone Axis=0.5
    Velocity Mode=::NewEnumerator2
    Velocity Speed=&RandomRangeFloat(Minimum=50, Maximum=500)
   )
   &SubUVAnimation(End Frame=3)
   &CameraOffset(Camera Offset Amount=44)
  ]
  particleUpdateScript [
   &ParticleState()
   &ScaleSpriteSize(Uniform Curve Sprite Scale=DataInterface() {propertyValues {Curve {keys [
        {interpMode=RCIM_Cubic, leaveTangent=5.6407, tangentMode=RCTM_Break, time=0.0134, value=0.811}
        {tangentMode=RCTM_User, time=0.4174, value=1.6577}
        <at 1.0035 2.0637>
       ]}}})
   &GravityForce(Gravity=<0 0 -1200>)
   &ScaleColor(
    Scale Alpha=&FloatFromCurve(FloatCurve=DataInterface() {propertyValues {Curve {keys [
         {arriveTangent=13.2056, interpMode=RCIM_Cubic, leaveTangent=13.2056, tangentMode=RCTM_User, time=0, value=0}
         {arriveTangent=-0.4588, interpMode=RCIM_Cubic, leaveTangent=-0.4588, tangentMode=RCTM_User, time=0.2, value=1}
         {arriveTangent=-0.0447, interpMode=RCIM_Cubic, tangentMode=RCTM_Break, time=1, value=0}
        ]}}})
    ScaleA=true
    ScaleRGB=true
   )
   &SolveForcesAndVelocity(Clamp Velocity=false, Limit Acceleration=false)
   &DynamicMaterialParameters(
    Index 0 Param 1=&FloatFromCurve(FloatCurve=DataInterface() {propertyValues {Curve {keys [
         {interpMode=RCIM_Cubic, leaveTangent=-0.7415, tangentMode=RCTM_Break, time=0, value=1}
         {arriveTangent=-1.5521, interpMode=RCIM_Cubic, leaveTangent=-8.0089, tangentMode=RCTM_Break, time=1, value=0}
        ]}}})
    Index 0 Param 2=&FloatFromCurve(
     FloatCurve=DataInterface() {propertyValues {Curve {keys [
         {interpMode=RCIM_Cubic, leaveTangent=-0.4439, tangentMode=RCTM_Break, time=0, value=0.2}
         {arriveTangent=-0.0393, interpMode=RCIM_Cubic, tangentMode=RCTM_Break, time=1, value=-0.3}
        ]}}}
     Scale Curve=0.7
    )
    Index 0 Param 3=&FloatFromCurve(
     FloatCurve=DataInterface() {propertyValues {Curve {keys [{interpMode=RCIM_Cubic, leaveTangent=2.6753, tangentMode=RCTM_Break, time=0.5, value=0}, <at 1 1>]}}}
     Scale Curve=3
    )
    Index 0 Param 4=&FloatFromCurve(FloatCurve=DataInterface() {propertyValues {Curve {keys [{interpMode=RCIM_Cubic, leaveTangent=2.6753, tangentMode=RCTM_Break, time=0, value=0}, <at 1 1>]}}})
    Index 1 Param 1=&FloatFromCurve(FloatCurve=DataInterface() {propertyValues {Curve {keys [{interpMode=RCIM_Cubic, tangentMode=RCTM_User, time=0, value=1}, {interpMode=RCIM_Cubic, tangentMode=RCTM_User, time=0.6, value=0}]}}})
    Module.Param0WriteEnabled=true
    Module.Param0WriteEnabled1=true
    Module.Param1WriteEnabled=true
    Module.Param1WriteEnabled1=true
    Module.Param2WriteEnabled=true
    Module.Param2WriteEnabled1=true
    Module.Param3WriteEnabled=true
    Module.Param3WriteEnabled1=true
    Write Parameter Index 1=true
   )
   &ScaleColor(
    Scale Alpha=&FNTimeOfDayScalar(Night Scale=0.8, Morning Scale=0.4, Evening Scale=0.5)
    Scale RGB=&MakeVectorFromLinearColorRGB(Color=&FNTimeOfDayColor(Night Color=<rgba 0.6 0.8 1 1>, Morning Color=<rgba 1 0.625 0.55 1>, Day Color=<rgba 1 1 1 1>, Evening Color=<rgba 1 0.5 0.4 1>))
    ScaleA=true
    ScaleRGB=true
   ) {enabled=false, module=ScaleColor001}
   &ScaleColor(
    Scale Alpha=$System.TOD_Alpha
    Scale RGB=&MakeVectorFromLinearColorRGB(Color=$System.TOD_Color)
    ScaleA=true
    ScaleRGB=true
   ) {module=ScaleColor002}
   &KillParticles(Kill Particles=&SetBoolByFloatComparison(A=$Particles.NormalizedAge, B=0.9))
  ]
  rendererClasses [&NiagaraSpriteRendererProperties]
  renderers [{rendererClass=&NiagaraSpriteRendererProperties}]
 }
 TopSplash002 {
  simTarget=GPUComputeSim
  emitterUpdateScript [
   &EmitterState(
    Enable Distance Culling=true
    Life Cycle Mode=::NewEnumerator1
    Loop Duration=2
    MaxDistance=10000
    Module.bUseMaxDistance=true
    Module.bUseMinDistance=false
    Scalability Mode=::NewEnumerator1
    UseLoopDelay=false
   )
   &SpawnRate(
    SpawnRate=&Multiply_Float(A=125, B=&Divide_Float(A=$User.FallWidth, B=850))
    Use Spawn Probability=false
   )
  ]
  particleSpawnScript [
   &PartitionParticles() {enabled=false}
   &InitializeParticle(
    Color Maximum=<rgba 1 1 1 1.5>
    Color Minimum=<rgba 1 1 1 1>
    Color Mode=::NewEnumerator2
    Lifetime Max=0.8
    Lifetime Min=0.4
    Lifetime Mode=::NewEnumerator1
    Mass Mode=::NewEnumerator3
    Sprite Rotation Mode=::NewEnumerator2
    Sprite Size Mode=::NewEnumerator1
    Uniform Sprite Size Max=30
    Uniform Sprite Size Min=10
    UsePositionOffset=false
    Write Material Random=true
   )
   &ShapeLocation(Box Size=<50 50 150>, MaskSpawnGroup=false, Override Seed=false, Shape Primitive=::NewEnumerator2, Surface Only=false, UseOwnerScale=true)
   /SetVariables(Particles.Position=&AddVectorToPosition(
     Position=$Particles.Position
     Vector=&Multiply_VectorByFloat(
      Vector=&TransformVector(Vector=<1 0 0>, Destination Space=::Simulation)
      Float=&Multiply_Float(
       A=&Lerp_Float(
        A=0
        B=&Divide_Float(A=$User.FallWidth, B=2)
        Alpha=&FloatFromCurve(
         FloatCurve=DataInterface() {propertyValues {Curve {keys [
             {interpMode=RCIM_Cubic, leaveTangent=3.3639, tangentMode=RCTM_Break, time=0, value=0}
             {arriveTangent=-0.0122, interpMode=RCIM_Cubic, tangentMode=RCTM_Break, time=1, value=1}
            ]}}}
         CurveIndex=&RandomRangeFloat(Minimum=0, Maximum=1)
         Scale Curve=0.8
        )
       )
       B=&Lerp_Float(
        A=-1
        B=1
        Alpha=&MakeFloatFromInt(Int=&RandomRangeInt(Minimum=0, Maximum=1))
       )
      )
     )
    )) {enabled=false, isSetParameters=true, module=SetVariables_#5}
   &AddVelocity(
    Cone Angle=300
    Cone Axis=&SelectVectorFromArray(
     Direct Array Index=$Particles.PartitionParticles.Partition
     Vector Selection Array=DataInterface() {propertyValues {FloatData [<-1 0 0.2>, <1 0 0.2>]}}
    )
    Distribution Along Cone Axis=0.25
    Inner Cone Angle=50
    Module.UseConeFalloff=true
    Module.UseVelDistribution=true
    Override Seed=false
    Speed Falloff From Cone Axis=0.5
    Velocity Mode=::NewEnumerator2
    Velocity Speed=&RandomRangeFloat(Minimum=250, Maximum=500)
   )
   &SubUVAnimation(End Frame=3)
   &CameraOffset(Camera Offset Amount=80)
  ]
  particleUpdateScript [
   &ParticleState()
   &ScaleSpriteSize(Uniform Curve Sprite Scale=DataInterface() {propertyValues {Curve {keys [
        {interpMode=RCIM_Cubic, leaveTangent=5.6407, tangentMode=RCTM_Break, time=0.0134, value=0.811}
        {tangentMode=RCTM_User, time=0.4174, value=1.6577}
        <at 1.0035 2.0637>
       ]}}})
   &GravityForce(Gravity=<0 0 -1200>)
   &ScaleColor(
    Scale Alpha=&FloatFromCurve(
     FloatCurve=DataInterface() {propertyValues {Curve {keys [
         {arriveTangent=13.2056, interpMode=RCIM_Cubic, leaveTangent=13.2056, tangentMode=RCTM_User, time=0, value=0}
         {arriveTangent=-0.4588, interpMode=RCIM_Cubic, leaveTangent=-0.4588, tangentMode=RCTM_User, time=0.2, value=1}
         {arriveTangent=-0.0447, interpMode=RCIM_Cubic, tangentMode=RCTM_Break, time=1, value=0}
        ]}}}
     Scale Curve=0.4
    )
    ScaleA=true
    ScaleRGB=true
   )
   &VortexForce(
    UseOriginOffset=false
    Vortex Force Amount=&SelectFloatFromArray(Direct Array Index=$Particles.PartitionParticles.Partition, Float Selection Array=DataInterface() {propertyValues {FloatData [1000,-1000]}})
    Vortex Origin:NiagaraPosition=$Particles.ShapeLocation.ShapePosition
   )
   &SolveForcesAndVelocity(Clamp Velocity=false, Limit Acceleration=false)
   &ScaleSpriteSizeBySpeed(Max Scale Factor=<1 5>, Min Scale Factor=<0.5 0.5>)
   &DynamicMaterialParameters(
    Index 0 Param 1=&FloatFromCurve(FloatCurve=DataInterface() {propertyValues {Curve {keys [
         {interpMode=RCIM_Cubic, leaveTangent=-0.7415, tangentMode=RCTM_Break, time=0, value=1}
         {arriveTangent=-1.5521, interpMode=RCIM_Cubic, leaveTangent=-8.0089, tangentMode=RCTM_Break, time=1, value=0}
        ]}}})
    Index 0 Param 2=&FloatFromCurve(
     FloatCurve=DataInterface() {propertyValues {Curve {keys [
         {interpMode=RCIM_Cubic, leaveTangent=-0.4439, tangentMode=RCTM_Break, time=0, value=0.2}
         {arriveTangent=-0.0393, interpMode=RCIM_Cubic, tangentMode=RCTM_Break, time=1, value=-0.3}
        ]}}}
     Scale Curve=0.7
    )
    Index 0 Param 3=&FloatFromCurve(
     FloatCurve=DataInterface() {propertyValues {Curve {keys [{interpMode=RCIM_Cubic, leaveTangent=2.6753, tangentMode=RCTM_Break, time=0.5, value=0}, <at 1 1>]}}}
     Scale Curve=3
    )
    Index 0 Param 4=&FloatFromCurve(FloatCurve=DataInterface() {propertyValues {Curve {keys [{interpMode=RCIM_Cubic, leaveTangent=2.6753, tangentMode=RCTM_Break, time=0, value=0}, <at 1 1>]}}})
    Index 1 Param 1=&FloatFromCurve(FloatCurve=DataInterface() {propertyValues {Curve {keys [{interpMode=RCIM_Cubic, tangentMode=RCTM_User, time=0, value=1}, {interpMode=RCIM_Cubic, tangentMode=RCTM_User, time=0.6, value=0}]}}})
    Module.Param0WriteEnabled=true
    Module.Param0WriteEnabled1=true
    Module.Param1WriteEnabled=true
    Module.Param1WriteEnabled1=true
    Module.Param2WriteEnabled=true
    Module.Param2WriteEnabled1=true
    Module.Param3WriteEnabled=true
    Module.Param3WriteEnabled1=true
    Write Parameter Index 1=true
   )
   &ScaleColor(
    Scale Alpha=&FNTimeOfDayScalar(Night Scale=0.8, Morning Scale=0.4, Evening Scale=0.5)
    Scale RGB=&MakeVectorFromLinearColorRGB(Color=&FNTimeOfDayColor(Night Color=<rgba 0.6 0.8 1 1>, Morning Color=<rgba 1 0.625 0.55 1>, Day Color=<rgba 1 1 1 1>, Evening Color=<rgba 1 0.5 0.4 1>))
    ScaleA=true
    ScaleRGB=true
   ) {enabled=false, module=ScaleColor001}
   &ScaleColor(
    Scale Alpha=$System.TOD_Alpha
    Scale RGB=&MakeVectorFromLinearColorRGB(Color=$System.TOD_Color)
    ScaleA=true
    ScaleRGB=true
   ) {module=ScaleColor002}
   &KillParticles(Kill Particles=&SetBoolByFloatComparison(A=$Particles.NormalizedAge, B=0.9)) {enabled=false}
  ]
  rendererClasses [&NiagaraSpriteRendererProperties]
  renderers [{rendererClass=&NiagaraSpriteRendererProperties}]
 }
}
summary {
 systemName=NS_CarWash_Roller_Water
 emitters [
  {emitterName=TopSmoke, rendererClasses [&NiagaraSpriteRendererProperties]}
  {emitterName=TopSplash, rendererClasses [&NiagaraSpriteRendererProperties]}
  {emitterName=TopSplash002, simTarget=GPUComputeSim, rendererClasses [&NiagaraSpriteRendererProperties]}
 ]
 userVariables [
  User.BottomPosition {type=NiagaraPosition, description="", defaultValue {}}
  User.BottomSmokeDensity {type=NiagaraFloat, defaultValue=1, description=""}
  User.FallWidth {type=NiagaraFloat, defaultValue=900, description=""}
  User.HitSplashEffectDensity {type=NiagaraFloat, defaultValue=1, description=""}
  User.RainbowStrength {type=NiagaraFloat, defaultValue=1, description=""}
  User.RippleOffset {type=NiagaraFloat, defaultValue=70, description=""}
  User.RippleOpacity {type=NiagaraFloat, defaultValue=1, description=""}
  User.RippleWidth {type=NiagaraFloat, defaultValue=10000, description=""}
  User.SpawnSplashPositionArray {
   type=NiagaraDataInterfaceArrayPosition
   description=""
   defaultValue {
    struct=NiagaraExt_VariableValue_DataInterface
    value {
     dataInterface=&self:NiagaraDataInterfaceArrayPosition_0
     dataInterfaceClass=&NiagaraDataInterfaceArrayPosition
    }
   }
  }
  User.TopPosition {type=NiagaraPosition, description="", defaultValue {}}
 ]
}
